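# Parent material for the head of the DuneBriefShift male medium player cosmetic, built on material attributes with the shading model chosen by expression.
# The shared 2023 character PreFX function builds the base material and masks.
# An eye animation function, set to use UV2, takes that material and the Tech Art mask, then feeds the shared PostFX function.
# The eye animation also drives the temporal responsiveness output.
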Material /BRCosmetics/Characters/Player/Male/Medium/Bodies/M_MED_DuneBriefShift_Get/Materials/M_DuneBriefShift_Get_Head_Paret
graph {
 properties {
  bUseMaterialAttributes=true
  bUsedWithClothing=true
  bUsedWithInstancedStaticMeshes=true
  bUsedWithMeshDeformer=true
  bUsedWithMorphTargets=true
  bUsedWithNanite=true
  bUsedWithSkeletalMesh=true
  pixelDepthOffsetMode=PDOM_Legacy
  previewMesh=/BRCosmetics/Characters/Player/Male/Medium/Heads/M_MED_DuneBriefMerry_Head/Meshes/M_MED_DuneBriefMerry_Head
  shadingModel=MSM_FromMaterialExpression
  usedShadingModels=Default Lit | Subsurface | Clear Coat | Subsurface Profile | Cloth
 }
 nodes [
  n1 = MaterialFunctionCall() {materialFunction=&MF_CharacterParent_2023_PreFX}
  n2 = MaterialFunctionCall!(
   Material=NamedRerouteDeclaration(None=@n1.Result) {name=Material, nodeColor=<rgba 0 1 0.1094 1>}
   Use UV2=StaticBool() {value=true}
   Mask=NamedRerouteDeclaration(None=@n1.Tech Art Mask) {name=TechArtMask, nodeColor=<rgba 0.2188 0 1 1>}
  ) {materialFunction=/Game/Global/Materials/Character_Parent_2023/Functions/MF_PBR_EyeAnimation_v1}
  MaterialFunctionCall(Material=@n2.Result, M=@n1.M, Pre Skinned Local Position=@n1.M, Distance Blend=@n1.M, Fresnel=@n1.M) {materialFunction=&MF_CharacterParent_2023_PostFX}
  NamedRerouteDeclaration(None=@n1.M) {name=Mask, nodeColor=<rgba 1 0 0.2109 1>}
  NamedRerouteDeclaration(None=@n1.Fresnel) {name=Fresnel, nodeColor=<rgba 1 0.6563 0 1>}
  TemporalResponsivenessOutput!(None=@n2.TemporalResponsiveness)
 ]
}
